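# Unlit emissive material for the Proto09Glitched vehicle wheel, showing a glitchy holographic code stream.
# A screen-space code texture, scaled by camera distance and resolution, pans in steps and is masked by normal-mapped Fresnel.
# Vertex red blends rim and tire; the tire adds an emissive texture with UVs jittered by sine waves and scrolling masks.
# CustomColor tints the result, which then passes through the shared vehicle damage and de-rez functions.
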Material /VehicleCosmetics/Wheels/Proto09Glitched/Materials/M_Wheel_Proto09Glitched
graph {
 properties {
  bAutomaticallySetUsageInEditor=false
  bUseMaterialAttributes=true
  bUsedWithSkeletalMesh=true
  shadingModel=MSM_Unlit
  usedShadingModels=Unlit
 }
 nodes [
  n1 = TextureSampleParameter2D() {
   parameterName=Proto09Glitched_Rim_N
   samplerType=SAMPLERTYPE_Normal
   texture=/VehicleCosmetics/Wheels/Proto09Glitched/Textures/T_Proto09Glitched_Rim_N
  }
  n2 = ScalarParameter() {defaultValue=2, parameterName=Fresnel_Sharpness}
  n3 = MaterialFunctionCall!(Input Normal=@n1.RGB, Input Sharpness=@n2, Input Two Sided) {materialFunction=/VehicleCosmetics/Wheels/CarCar_Holographic/Materials/MF_Fresnel_Advanced}
  n4 = ScalarParameter() {defaultValue=0.3, parameterName=Fresnel_Occlusion}
  n5 = Add(A=@n3.Edge, B=@n4)
  n6 = Distance(A=WorldPosition().XYZ, B=CameraPositionWS())
  n7 = ScalarParameter() {defaultValue=0.003, parameterName=Texture_size}
  n8 = ScreenPosition()
  n9 = Constant() {r=0.5}
  n10 = Add(A=Multiply(A=Multiply(A=@n6, B=@n7), B=Subtract(A=@n8.ViewportUV, B=@n9)), B=@n9)
  n11 = Subtract(A=ViewSize(), B=Constant() {r=1250})
  n12 = Multiply(
   A=Divide(A=ComponentMask(None=@n10) {g=true, r=true}, B=Constant() {r=1})
   B=LinearInterpolate(A=Constant() {r=10}, B=Constant() {r=25}, Alpha=Clamp(None=@n11))
  )
  n13 = Constant2Vector() {g=1, r=1}
  n14 = Time()
  n15 = Constant() {r=5}
  n16 = Constant() {r=3}
  n17 = Floor(None=Fmod(A=Multiply(A=@n14, B=@n15), B=@n16))
  n18 = TextureSample(UVs=Panner(Coordinate=Divide(A=@n12, B=@n13), Time=Divide(A=@n17, B=@n16)) {speedX=0.8, speedY=0.5}) {samplerType=SAMPLERTYPE_LinearColor, texture=/VehicleCosmetics/Wheels/Proto09Glitched/Textures/T_Proto09Glitched_Tire_RGB}
  n19 = Multiply(A=ScalarParameter() {defaultValue=1, parameterName=CodeStreamIntensity}, B=@n18.R)
  n20 = ScalarParameter() {defaultValue=0.3, parameterName=Fresnel_Overlay}
  n21 = Add(A=Multiply(A=Clamp(None=@n5), B=@n19), B=Multiply(A=@n3.Edge, B=@n20))
  n22 = Constant()
  n23 = ConstantBiasScale(None=@n18.A) {bias=-0.8, scale=2}
  n24 = Multiply(A=@n19, B=Multiply(A=Clamp(None=@n23), B=ScalarParameter() {defaultValue=0.7, parameterName=CodeHighlght_Intensity}))
  n25 = TextureSampleParameter2D() {
   parameterName=Proto09Glitched__Tire_N
   samplerType=SAMPLERTYPE_Normal
   texture=/VehicleCosmetics/Wheels/Proto09Glitched/Textures/T_Proto09Glitched_Tire_N
  }
  n26 = MaterialFunctionCall(Input Normal=@n25.RGB, Input Sharpness=@n2) {materialFunction=&MF_Fresnel_Advanced}
  n27 = Multiply(A=Clamp(None=Add(A=@n4, B=@n26.Edge)), B=@n19)
  n28 = Constant2Vector()
  n29 = Constant2Vector() {g=0.08, r=0.03}
  n30 = Time()
  n31 = Multiply(A=@n30, B=Constant() {r=3})
  n32 = Constant() {r=0.15}
  n33 = Sine(None=Multiply(A=@n30, B=Constant() {r=0.66}))
  n34 = Multiply(A=Sine(None=Multiply(A=@n30, B=@n32)) {period=0.05}, B=@n33)
  n35 = Constant() {r=0.6}
  n36 = Multiply(A=Sine(None=Multiply(A=@n30, B=@n35)) {period=2}, B=@n33)
  n37 = Multiply(A=Multiply(A=Sine(None=@n31), B=@n34), B=Multiply(A=@n34, B=@n36))
  n38 = TextureCoordinate()
  n39 = TextureSample(UVs=Panner(Coordinate=TextureCoordinate() {vTiling=2}, Time=@n30) {speedX=0.2, speedY=0.1}) {samplerType=SAMPLERTYPE_LinearColor, texture=/VehicleCosmetics/Wheels/Proto09Glitched/Textures/T_Proto09Glitched_TiledGradient_Balanced}
  n40 = LinearInterpolate(
   A=Add(A=LinearInterpolate(A=@n28, B=@n29, Alpha=@n37), B=@n38)
   B=@n38
   Alpha=Multiply(A=ConstantBiasScale(None=@n39.R) {bias=-0.5, scale=20}, B=Ceil(None=@n36))
  )
  n41 = ScreenPosition()
  n42 = Constant() {r=0.2}
  n43 = Panner(Coordinate=ComponentMask(None=Divide(A=@n41.ViewportUV, B=@n42)) {g=true, r=true}) {speedX=-1}
  n44 = Clamp(None=ConstantBiasScale(None=TextureSample(UVs=@n43) {samplerType=SAMPLERTYPE_LinearColor, texture=&T_Proto09Glitched_Tire_RGB}.A) {bias=-0.2, scale=1})
  n45 = Add(
   A=Add(A=Add(A=@n24, B=@n27), B=Multiply(A=@n20, B=@n26.Edge))
   B=TextureSample(UVs=LinearInterpolate(A=@n40, B=@n38, Alpha=@n44)) {samplerType=SAMPLERTYPE_LinearColor, texture=/VehicleCosmetics/Wheels/Proto09Glitched/Textures/T_Proto09Glitched_Tire_E}.R
  )
  n46 = Multiply(A=Constant() {r=0.04}, B=@n44)
  n47 = Multiply(
   A=LinearInterpolate(A=Add(A=@n21, B=@n22), B=Add(A=@n45, B=@n46), Alpha=VertexColor().R)
   B=VectorParameter() {parameterName=CustomColor, defaultValue=<rgba 1 1 1 1>}.RGB
  )
  MaterialFunctionCall(In=MaterialFunctionCall(Material=SetMaterialAttributes!(MaterialAttributes, Emissive Color=@n47, Opacity) {attributeSetTypes [#1, #2]}) {materialFunction=&MF_BaseVehicle_Damage}.Result) {materialFunction=&MF_VehicleDeRez_Transition}
 ]
}
