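# Translucent unlit effect material for the Proto09 Glitched vehicle wheel, showing animated tri-color twirls that intermittently glitch.
# A stepped Noise1D-over-time switch blends counter-rotating, noise-jittered twirls into a desaturated, vertically panned version with scrolling gradient bands.
# Scrolling rainbow-curve lookups, indexed by a half-UV texture, mix the three custom colors.
# A radial UV mask shapes emissive and opacity before vehicle damage and de-rez functions are applied.
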Material /VehicleCosmetics/Wheels/Proto09Glitched/Materials/M_Wheel_Proto09Glitched_Effect
graph {
 properties {
  bAutomaticallySetUsageInEditor=false
  bUseMaterialAttributes=true
  bUsedWithMeshDeformer=true
  bUsedWithSkeletalMesh=true
  blendMode=BLEND_Translucent
  shadingModel=MSM_Unlit
  usedShadingModels=Unlit
 }
 nodes [
  n1 = Time()
  n2 = Sine(None=Multiply(A=@n1, B=Constant() {r=0.3}))
  n3 = Constant() {r=0.2}
  n4 = Sine(None=Multiply(A=Add(A=@n1, B=@n3), B=Constant() {r=0.1}))
  n5 = Add(A=Multiply(A=Time(), B=Constant() {r=0.1}), B=Multiply(A=Add(A=@n2, B=@n4), B=Constant() {r=0.1}))
  n6 = ScalarParameter() {defaultValue=1, parameterName=Vertical_Panning_Intensity}
  n7 = Constant() {r=10}
  n8 = Multiply(A=Multiply(A=Sine(None=@n5), B=Divide(A=@n6, B=@n7)), B=Multiply(A=Time(), B=Constant() {r=0.1}))
  n9 = Panner(Time=@n8) {speedY=0.1}
  n10 = Time()
  n11 = Constant() {r=0.4}
  n12 = Fmod(
   A=MaterialFunctionCall!(In=Multiply(A=@n10, B=@n11)) {materialFunction=/VehicleCosmetics/Wheels/NewtonBall/Materials/Noise1D}.Result
   B=Constant() {r=5}
  )
  n13 = Clamp(None=Multiply(A=Floor(None=@n12), B=Constant() {r=1}))
  n14 = Multiply(
   A=LinearInterpolate(
    A=TextureSample() {texture=/VehicleCosmetics/Wheels/_SharedTextures/FX/Proto09_Glitched_HalfUV}.RGB
    B=TextureSample(UVs=@n9) {texture=&Proto09_Glitched_HalfUV}.RGB
    Alpha=@n13
   )
   B=Constant() {r=2}
  )
  n15 = Clamp(None=Abs(None=DotProduct(A=@n14, B=@n14)))
  n16 = Panner(Coordinate=Multiply(A=@n15, B=Constant() {r=3})) {speedX=-0.2}
  n17 = Constant() {r=2}
  n18 = TextureSample(UVs=Panner(Coordinate=Multiply(A=@n15, B=@n17)) {speedX=-0.2}) {texture=/VehicleCosmetics/Wheels/_SharedTextures/FX/Curve_Rainbow}
  n19 = Add(A=@n18.R, B=@n18.R)
  n20 = Add(A=@n18.R, B=@n18.R)
  n21 = LinearInterpolate(
   A=LinearInterpolate(
    A=VectorParameter() {parameterName=CustomColor, defaultValue=<rgba 0.18 0.76 0.83 1>}.RGB
    B=VectorParameter() {parameterName=CustomColor3, defaultValue=<rgba 0.18 0.87 0.18 1>}.RGB
    Alpha=TextureSample(UVs=@n16) {texture=&Curve_Rainbow}.G
   )
   B=VectorParameter() {parameterName=CustomColor2, defaultValue=<rgba 0.87 0.28 0.25 1>}.RGB
   Alpha=Multiply(A=Multiply(A=@n19, B=@n20), B=Add(A=@n18.G, B=@n18.G))
  )
  n22 = LinearInterpolate(A=@n21, B=Desaturation(None=@n21, Fraction=Constant() {r=0.7}), Alpha=@n13)
  n23 = Time()
  n24 = Constant() {r=2}
  n25 = TextureSample(UVs=Panner(Coordinate=TextureCoordinate() {uTiling=0.5, vTiling=0.1}) {speedX=0.1}) {texture=/VehicleCosmetics/Wheels/_SharedTextures/FX/Universal_Noise_42_RGB}
  n26 = Multiply(
   A=MaterialFunctionCall(In=Multiply(A=@n23, B=@n24)) {materialFunction=&Noise1D}.Result
   B=Clamp(None=ConstantBiasScale(None=@n25.R) {bias=-0.4, scale=1})
  )
  n27 = Multiply(A=Time(), B=Constant() {r=2})
  n28 = Constant() {r=2}
  n29 = Divide(A=Floor(None=Fmod(A=@n27, B=@n28)), B=@n28)
  n30 = Add(A=TextureCoordinate(), B=Multiply(A=Multiply(A=@n26, B=@n29), B=Constant() {r=1}))
  n31 = Add(
   A=TextureSample(UVs=Rotator(Coordinate=@n30) {speed=1}) {texture=/VehicleCosmetics/Wheels/Proto09Glitched/Textures/T_Proto09Glitched_Twirl_RGB}.R
   B=TextureSample(UVs=Rotator(Coordinate=@n30) {speed=-2}) {texture=&T_Proto09Glitched_Twirl_RGB}.G
  )
  n32 = TextureCoordinate() {vTiling=3}
  n33 = TextureSample(UVs=Add(A=Panner(Coordinate=@n32) {speedX=2, speedY=0.2}, B=@n8)) {texture=/VehicleCosmetics/Wheels/_SharedTextures/FX/TiledGradient_Balanced}
  n34 = Clamp(None=ConstantBiasScale(None=@n33.R) {bias=-0.5, scale=20})
  n35 = Constant() {r=0.1}
  n36 = TextureSample(UVs=Add(A=Multiply(A=@n34, B=@n35), B=@n9)) {texture=&T_Proto09Glitched_Twirl_RGB}
  n37 = Add(A=@n36.R, B=@n36.R)
  n38 = Clamp(None=ConstantBiasScale(None=@n37) {bias=0.3, scale=1})
  n39 = LinearInterpolate(A=Clamp(None=ConstantBiasScale(None=@n31) {bias=0.3, scale=1}), B=@n38, Alpha=@n13)
  n40 = Add(A=Multiply(A=@n22, B=Clamp(None=@n39)), B=Multiply(A=@n31, B=Constant() {r=0.1}))
  n41 = TextureCoordinate() {uTiling=2, vTiling=5}
  n42 = ConstantBiasScale(None=TextureSample(UVs=Panner(Coordinate=@n41) {speedX=4, speedY=1}) {texture=&TiledGradient_Balanced}.R)
  n43 = TextureSample(UVs=Add(A=@n9, B=Constant() {r=0.5})) {samplerType=SAMPLERTYPE_LinearColor, texture=/VehicleCosmetics/Wheels/_SharedTextures/FX/BasicGradient}
  n44 = Add(A=Multiply(A=Clamp(None=@n42), B=@n43.R), B=Multiply(A=@n43.R, B=@n33.R))
  n45 = Add(A=Add(A=Multiply(A=@n22, B=@n38), B=Clamp(None=@n44)), B=Multiply(A=@n37, B=Constant() {r=0.1}))
  n46 = Add(A=TextureCoordinate(), B=Constant() {r=-0.5})
  n47 = OneMinus(None=Abs(None=DotProduct(A=@n46, B=@n46)))
  n48 = Constant() {r=-0.78}
  n49 = Clamp(None=Multiply(A=Add(A=@n47, B=@n48), B=Constant() {r=3}))
  n50 = Constant() {r=0.2}
  n51 = Multiply(
   A=Multiply(A=LinearInterpolate(A=@n40, B=@n45, Alpha=@n13), B=Power(Base=@n49, Exp=@n50))
   B=ScalarParameter() {defaultValue=1, parameterName=Emissive_Intensity}
  )
  n52 = Clamp(None=Multiply(A=Multiply(A=@n49, B=@n49), B=ScalarParameter() {defaultValue=15, parameterName=Opacity_Hardness}))
  MaterialFunctionCall(In=MaterialFunctionCall(Material=SetMaterialAttributes!(MaterialAttributes, Emissive Color=@n51, Opacity=@n52) {attributeSetTypes [#1, #2]}) {materialFunction=&MF_BaseVehicle_Damage}.Result) {materialFunction=&MF_VehicleDeRez_Transition}
 ]
}
